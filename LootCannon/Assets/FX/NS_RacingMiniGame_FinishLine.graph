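# Niagara system for the racing minigame finish line: a glowing sprite plane with sparks drifting inside a box, both sized from System.FinishLineSize.
# The plane fades in with system age.
# A system Timeline drives the outro: both flash brighter, the plane fades out and sparks surge upward, shrink and stop spawning.
# Camera-distance fades scale brightness, plane material and spark size; sparks add a second sprite renderer with glow parameters.
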NiagaraSystem /LootCannon/Assets/FX/NS_RacingMiniGame_FinishLine
dependencies {
 usedDataInterfaces [&NiagaraDataInterfaceColorCurve, &NiagaraDataInterfaceCamera, &NiagaraDataInterfaceCurve]
 usedDynamicInputs [&Boolean_NotOperation, /Game/Effects/Niagara/DynamicInputs/State/FN_IsExecutionStateActive]
 usedModules [
  &self:SystemSpawnScript.SystemScriptSource.SystemScriptGraph.NiagaraNodeAssignment_0.SetVariables_#1
  &self:SystemSpawnScript.SystemScriptSource.SystemScriptGraph.NiagaraNodeAssignment_1.SetVariables_#2
  &SystemState
  &Timeline
 ]
}
emitters {
 NE_FinishlinePlane {
  emitterUpdateScript [
   &EmitterState(Life Cycle Mode=::NewEnumerator1, Loop Behavior=::NewEnumerator1, Loop Duration Mode=::NewEnumerator1, UseLoopDelay=false)
   &SpawnBurst_Instantaneous(Use Spawn Probability=false, UseLoopCountLimit=false)
  ]
  particleSpawnScript [&InitializeParticle(
    Color=<rgba 1 0.5679 0 1>
    Lifetime=5
    Position Offset=<0 0 -40>
    Sprite Size=&MakeVector2D(
     X=&MakeFloatFromVector(Vector=$System.FinishLineSize, Channel=::NewEnumerator2)
     Y=&MakeFloatFromVector(Vector=$System.FinishLineSize, Channel=::NewEnumerator0)
    )
    Sprite Size Mode=::NewEnumerator4
    UsePositionOffset=true
    Write Material Random=true
   )]
  particleUpdateScript [
   &ParticleState(Kill Particles When Lifetime Has Elapsed=false)
   &SpriteFacingAndAlignment(Sprite Alignment=<-1 0 0>, Sprite Facing=<0 1 0>, WriteAlignment=true, WriteFacing=true)
   &ScaleColor(
    Scale Alpha=&Multiply_Float(
     A=&Multiply_Float(
      A=&FloatFromCurve(
       FloatCurve=DataInterface() {propertyValues {Curve {keys [
           {arriveTangent=1.0686, interpMode=RCIM_Cubic, leaveTangent=1.0686, tangentMode=RCTM_User, time=0, value=0}
           {arriveTangent=0.2335, interpMode=RCIM_Cubic, leaveTangent=0.2335, tangentMode=RCTM_User, time=1.5, value=1}
          ]}}}
       CurveIndex=$Engine.System.Age
      )
      B=&FloatFromCurve(
       FloatCurve=DataInterface() {propertyValues {Curve {keys [
           {arriveTangent=-2.1931, interpMode=RCIM_Cubic, leaveTangent=-2.1931, tangentMode=RCTM_User, time=0.1, value=1}
           {arriveTangent=-0.1386, interpMode=RCIM_Cubic, leaveTangent=-0.1386, tangentMode=RCTM_User, time=1, value=0}
          ]}}}
       CurveIndex=$System.Timeline.NormalizedTime
      )
     )
     B=&Lerp_Float(A=0.5, B=2, Alpha=$System.FinishLineCameraFade)
    )
    Scale RGB=&MakeVectorFromLinearColorRGB(Color=&Multiply_LinearColorByFloat(
      Linear Color=&ColorFromCurve(
       ColorCurve=DataInterface() {propertyValues {
         AlphaCurve {keys [<at 0 1 RCIM_Cubic>]}
         BlueCurve {keys [<at 0 1 RCIM_Cubic>,<at 0.0323 3>,<at 1 1 RCIM_Cubic>]}
         GreenCurve {keys [<at 0 1 RCIM_Cubic>,<at 0.0323 3>,<at 1 1 RCIM_Cubic>]}
         RedCurve {keys [<at 0 1 RCIM_Cubic>,<at 0.0323 3>,<at 1 1 RCIM_Cubic>]}
        }}
       CurveIndex=$System.Timeline.NormalizedTime
      )
      Float=&Lerp_Float(
       A=1
       B=5
       Alpha=/Niagara/DynamicInputs/Camera/CameraFade(
        Camera Query:NiagaraDataInterfaceCamera=DataInterface()
        Fade Offset:NiagaraFloat=2000
        Fade Distance:NiagaraFloat=15000
        Use Owner Position:NiagaraBool=true
       )
      )
     ))
    ScaleA=true
    ScaleRGB=true
   )
   &DynamicMaterialParameters(
    Index 0 Param 1=&Lerp_Float(
     A=1
     B=6
     Alpha=&CameraFade(
      Camera Query:NiagaraDataInterfaceCamera=DataInterface()
      Fade Offset:NiagaraFloat=2000
      Fade Distance:NiagaraFloat=15000
      Use Owner Position:NiagaraBool=true
     )
    )
    Module.Index0Float0Write=true
    Module.Index0Float1Write=true
    Module.Index0Float2Write=true
    Module.Index0Float3Write=true
    Module.Index1Float0Write=true
    Module.Index1Float1Write=true
    Module.Index1Float2Write=true
    Module.Index1Float3Write=true
    Write Parameter Index 1=::NewEnumerator1
   )
   &KillParticles(Kill Particles=&SetBoolByFloatComparison(A=$System.Timeline.NormalizedTime, B=0.99))
  ]
  rendererClasses [&NiagaraSpriteRendererProperties]
  renderers [{rendererClass=&NiagaraSpriteRendererProperties}]
 }
 NE_Sparks {
  emitterUpdateScript [
   &EmitterState()
   &SpawnRate(
    SpawnRate=&MakeCustomFloatFromBool(Bool=&SetBoolByFloatComparison(A=$System.Timeline.NormalizedTime, B=0.99), True Float=0, False Float=20)
    Use Spawn Probability=false
   )
  ]
  particleSpawnScript [
   &InitializeParticle(
    Color=&ColorFromCurve(
     ColorCurve=DataInterface() {propertyValues {
       AlphaCurve {keys [<at 0 1 RCIM_Cubic>]}
       BlueCurve {keys [<at 0.4971 0.1994 RCIM_Cubic>,<at 0.5451 0.0102 RCIM_Cubic>]}
       GreenCurve {keys [<at 0.4971 0.5453 RCIM_Cubic>,<at 0.5451 0.4997 RCIM_Cubic>]}
       RedCurve {keys [<at 0.4971 1 RCIM_Cubic>,<at 0.5451 1 RCIM_Cubic>]}
      }}
     CurveIndex=&RandomRangeFloat(Minimum=0, Maximum=1)
    )
    Lifetime=&RandomRangeFloat(Minimum=1.5, Maximum=2)
    Sprite Size=&Multiply_Vector2D_ByFloat(Float=&RandomRangeFloat(Minimum=18, Maximum=25), Vector2D=<1 2>)
    Sprite Size Mode=::NewEnumerator4
    UsePositionOffset=false
    Write Material Random=true
   )
   &ShapeLocation(
    Box Size=&Multiply_Vector(A=$System.FinishLineSize, B=<1 1 0.25>)
    MaskSpawnGroup=false
    Override Seed=false
    Shape Primitive=::NewEnumerator2
    Surface Only=false
    UseOwnerScale=true
   )
  ]
  particleUpdateScript [
   &ParticleState()
   &ScaleSpriteSize(
    Uniform Curve Scale=&FloatFromCurve(
     FloatCurve=DataInterface() {propertyValues {Curve {keys [
         {interpMode=RCIM_Cubic, tangentMode=RCTM_User, time=0, value=1}
         {arriveTangent=-1.94, interpMode=RCIM_Cubic, leaveTangent=-1.94, tangentMode=RCTM_User, time=1, value=0}
        ]}}}
     CurveIndex=$System.Timeline.NormalizedTime
     Scale Curve=&Lerp_Float(
      A=0.6
      B=2
      Alpha=&CameraFade(
       Camera Query:NiagaraDataInterfaceCamera=DataInterface()
       Fade Offset:NiagaraFloat=2000
       Fade Distance:NiagaraFloat=15000
       Use Owner Position:NiagaraBool=true
      )
     )
    )
    Uniform Curve Sprite Scale=DataInterface() {propertyValues {Curve {keys [
        {interpMode=RCIM_Cubic, tangentMode=RCTM_User, time=0, value=0}
        {interpMode=RCIM_Cubic, tangentMode=RCTM_User, time=0.15, value=1}
        {interpMode=RCIM_Cubic, tangentMode=RCTM_User, time=0.75, value=1}
        {interpMode=RCIM_Cubic, tangentMode=RCTM_User, time=1, value=0}
       ]}}}
   )
   &CurlNoiseForce(
    Bias Noise=false
    Noise Strength=&Lerp_Float(
     A=250
     B=1550
     Alpha=&FloatFromCurve(
      FloatCurve=DataInterface() {propertyValues {Curve {keys [
          {arriveTangent=5.036, interpMode=RCIM_Cubic, leaveTangent=5.036, tangentMode=RCTM_User, time=0, value=0}
          {interpMode=RCIM_Cubic, tangentMode=RCTM_User, time=0.1, value=1}
         ]}}}
      CurveIndex=$System.Timeline.NormalizedTime
     )
    )
    Pan Noise=false
   )
   &GravityForce(Gravity=<0 0 150>)
   &AccelerationForce(Acceleration=&Lerp_Vector(
     A=<0 0 0>
     B=<0 0 1500>
     Alpha=&FloatFromCurve(FloatCurve=DataInterface() {propertyValues {Curve {keys [<at 0 0>,<at 0.1 1>]}}}, CurveIndex=$System.Timeline.NormalizedTime)
    ))
   &SolveForcesAndVelocity(Clamp Velocity=false, Limit Acceleration=false)
   /SetVariables(Particles.SparkGlowColor:LinearColor=&MultiplyLinearColor(
     Linear Color=<rgba 0.05 0.05 0.05 0.01>
     Scale Factor=&MakeVector4_FromLinearColor(LinearColor=$Particles.Color)
    )) {isSetParameters=true}
   /SetVariables(Particles.SparkGlowSize:Vector2f=&Multiply_Vector2D(A=$Particles.SpriteSize, B=<10 2>)) {isSetParameters=true}
   /SetVariables(Particles.SparkGlowDynMat:Vector4f=<2 20 90 1>) {isSetParameters=true}
   &ScaleColor(
    Scale RGB=&Multiply_VectorByFloat(
     Vector=&MakeVectorFromLinearColorRGB(Color=&ColorFromCurve(
       ColorCurve=DataInterface() {propertyValues {
         AlphaCurve {keys [<at 0 1 RCIM_Cubic>]}
         BlueCurve {keys [<at 0 1 RCIM_Cubic>,<at 0.0943 20>,<at 1 1 RCIM_Cubic>]}
         GreenCurve {keys [<at 0 1 RCIM_Cubic>,<at 0.0943 20>,<at 1 1 RCIM_Cubic>]}
         RedCurve {keys [<at 0 1 RCIM_Cubic>,<at 0.0943 20>,<at 1 1 RCIM_Cubic>]}
        }}
       CurveIndex=$System.Timeline.NormalizedTime
      ))
     Float=&Lerp_Float(
      A=1
      B=2
      Alpha=&CameraFade(
       Camera Query:NiagaraDataInterfaceCamera=DataInterface()
       Fade Offset:NiagaraFloat=2000
       Fade Distance:NiagaraFloat=15000
       Use Owner Position:NiagaraBool=true
      )
     )
    )
    ScaleA=true
    ScaleRGB=true
   )
  ]
  rendererClasses [&NiagaraSpriteRendererProperties]
  renderers [
   {rendererClass=&NiagaraSpriteRendererProperties}
   {rendererClass=&NiagaraSpriteRendererProperties, rendererIndex=1}
  ]
 }
}
summary {
 systemName=NS_RacingMiniGame_FinishLine
 emitters [
  {emitterName=NE_FinishlinePlane, rendererClasses [&NiagaraSpriteRendererProperties]}
  {emitterName=NE_Sparks, rendererClasses [&NiagaraSpriteRendererProperties]}
 ]
 userVariables []
}
