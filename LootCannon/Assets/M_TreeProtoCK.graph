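# M_TreeProtoCK is a two-sided Default Lit material, apparently for prototype trees in LootCannon, flagged for use with Nanite meshes.
# Its graph holds a single vector parameter, Color, with a dark green default that material instances can override.
# No textures, math or connection from that parameter to a material output appear in the graph, and automatic usage flagging is off.
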Material /LootCannon/Assets/M_TreeProtoCK
graph {
 properties {bAutomaticallySetUsageInEditor=false, bUsedWithNanite=true, twoSided=true, usedShadingModels=Default Lit}
 nodes [VectorParameter() {parameterName=Color, defaultValue=<rgba 0.0051 0.1198 0.0279 1>}]
}
